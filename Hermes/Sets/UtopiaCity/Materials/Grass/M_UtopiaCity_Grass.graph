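# Subsurface, Nanite-enabled grass ground material for UtopiaCity that can fake blade depth with parallax.
# Switchable parallax occlusion mapping ray-marches a height texture on world or mesh UVs, with optional flowmap distortion, pixel depth offset and quality-based fallbacks.
# Noise-blended colors and mowing-style stripes tint it, a ShadowReplace offset shifts shadow-pass positions, and a BasicGrass switch selects a textured floor instead.
# A separate custom jittered ray-march builds an alternative blade layer.
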Material /Hermes/Sets/UtopiaCity/Materials/Grass/M_UtopiaCity_Grass
graph {
 properties {
  bAutomaticallySetUsageInEditor=false
  bFullyRough=true
  bTangentSpaceNormal=false
  bUseMaterialAttributes=true
  bUsedWithNanite=true
  shadingModel=MSM_Subsurface
  usedShadingModels=Subsurface
 }
 nodes [
  Constant3Vector() {constant=<rgba 0.5 0 0 0>}
  Constant() {r=8}
  Constant() {r=20}
  Constant() {desc=DepthMax, r=0.05}
  Constant()
  Constant() {r=0.02}
  Constant() {r=24}
  Constant() {r=0.02}
  Constant() {r=1}
  Constant() {r=20}
  Constant()
  Constant() {r=0.3}
  Constant() {r=0.03}
  Constant() {r=0.015}
  Constant() {r=18}
  Constant() {r=300}
  Constant() {r=0.03}
  Constant() {r=600}
  Constant() {r=0.01}
  n1 = NamedRerouteUsage()
  DDX(Value=@n1)
  DDY(Value=@n1)
  n2 = TextureObject() {texture=&/T_UtopiaCity_ParallaxGrass}
  n3 = ScalarParameter() {defaultValue=12, parameterName=PDO_MaxSteps_Epic}
  n4 = ScalarParameter() {defaultValue=10, parameterName=PDO_MaxSteps_High}
  n5 = WorldPosition() {worldPositionShaderOffset=WPT_ExcludeAllShaderOffsets}
  n6 = ScalarParameter() {defaultValue=300, parameterName=Grass_UV_WorldScale_Distort}
  n7 = ComponentMask(None=TextureSample(UVs=Divide(A=@n5.XY, B=@n6)) {samplerType=SAMPLERTYPE_LinearColor, texture=/Game/Packages/Fortnite_UniqueMaterials_1/Textures/FlowMap}.RGB) {g=true, r=true}
  n8 = Add(
   A=Length(None=WorldPosition() {worldPositionShaderOffset=WPT_CameraRelative}.XY) {is=class MaterialExpressionLength}
   B=ScalarParameter() {defaultValue=-1000, parameterName=Grass_UV_WorldScale_FadeDistance}
  ) {constB=-1000, desc=Sets Blend Start Distance}
  n9 = ScalarParameter() {defaultValue=300, parameterName=Grass_UV_WorldScale_FadeSize}
  n10 = OneMinus(None=Saturate(None=Divide(A=@n8, B=@n9) {constB=100, desc=Sets Blend Range}))
  n11 = ScalarParameter() {parameterName=Grass_UV_WorldScale_FadeMin}
  n12 = Multiply(
   A=Multiply(A=Add(A=@n7) {constB=-0.5}) {constB=1.5}
   B=Multiply(A=ScalarParameter() {defaultValue=0.03, parameterName=Grass_UV_WorldScale_Distort_Mult}, B=Max(A=@n10, B=@n11) {constB=0.5})
  )
  n13 = Constant()
  n14 = StaticBoolParameter() {parameterName=UseDistortion}
  n15 = MaterialFunctionCall!(
   Heightmap Texture=@n2
   Height Ratio=ScalarParameter() {defaultValue=0.015, parameterName=PDO_HeightRatio}
   Min Steps=ScalarParameter() {defaultValue=4, parameterName=PDO_MinSteps}
   Max Steps=LinearInterpolate(
    A=ScalarParameter() {defaultValue=4, parameterName=PDO_MaxSteps_Far}
    B=QualitySwitch(Default=@n3, High=@n4, Epic=@n3)
    Alpha=NamedRerouteUsage()
   )
   UVs=Add(A=NamedRerouteUsage(), B=StaticSwitch(True=@n12, False=@n13, Value=@n14))
   Heightmap Channel
   Reference Plane=ScalarParameter() {parameterName=PDO_RefPlane}
   -----------------------------
   Use World Coordinates=NamedRerouteUsage()
   Specify Manual Texture Size
   Manual Texture Size
   ------------------------------
   "Render Shadows (Occlusion Mapping)"=StaticBool()
   Light Vector=AtmosphericLightVector()
   Shadow Steps
   Shadow Penumbra
   Transform To VertexNormal
  ) {materialFunction=/Engine/Functions/Engine_MaterialFunctions01/Texturing/ParallaxOcclusionMapping}
  Multiply(A=ComponentMask(None=@n15.Offset Only) {r=true})
  Multiply(A=@n15.Material Complexity - Steps Debug)
  MaterialExpressionRemap_3 = MaterialXRemap!(None=NamedRerouteUsage(), InputLow, InputHigh, TargetLow, TargetHigh) {targetLowDefault=0.75}
  MaterialExpressionRemap_0 = MaterialXRemap(None=NamedRerouteUsage()) {inputHighDefault=0.03}
  n16 = LinearInterpolate(
   A=Constant3Vector() {constant=<rgba 0.7 0.7 0.7 1>}
   B=Constant3Vector() {constant=<rgba 1.05 1.05 1.05 1>}
   Alpha=@MaterialExpressionRemap_0
  )
  n17 = Constant3Vector() {constant=<rgba 0.0566 0.1719 0.0143 1>}
  Multiply(A=NamedRerouteDeclaration(None=Multiply(A=@MaterialExpressionRemap_3, B=@n16)) {name=BladeGradient, nodeColor=<rgba 0.6406 0 1 1>}, B=@n17)
  n18 = ScalarParameter() {defaultValue=0.03, parameterName=DepthMax}
  Multiply(A=@n18, B=LinearInterpolate(Alpha=NamedRerouteUsage()) {constA=0.25})
  MaterialExpressionNamedRerouteUsage_14 = NamedRerouteUsage()
  MaterialExpressionNamedRerouteUsage_15 = NamedRerouteUsage()
  n19 = NamedRerouteUsage()
  n20 = Constant() {r=0.33}
  n21 = Constant() {r=1}
  MaterialExpressionLinearInterpolate_3 = LinearInterpolate(
   A=Constant() {r=1}
   B=Constant() {r=20}
   Alpha=Multiply(
    A=NamedRerouteUsage()
    B=MaterialFunctionCall(Input=@n19, Input Low=@n20, Input High=@n21) {materialFunction=&RemapValue_0-1}.Result
   )
  ) {constA=1, constB=8}
  MaterialExpressionTextureObject_0 = TextureObject() {texture=&T_UtopiaCity_ParallaxGrass}
  MaterialExpressionConstant_32 = Constant() {r=1}
  MaterialExpressionConstant_15 = Constant()
  MaterialExpressionScalarParameter_8 = ScalarParameter() {defaultValue=0.02, parameterName=SP_Depth}
  MaterialExpressionScreenPosition_1 = ScreenPosition()
  MaterialExpressionConstant_10 = Constant()
  MaterialExpressionConstant_9 = Constant() {r=1}
  MaterialExpressionScalarParameter_9 = ScalarParameter() {defaultValue=0.1, parameterName=SP_Offset}
  MaterialExpressionConstant3Vector_4 = Constant3Vector()
  MaterialExpressionConstant_29 = Constant()
  n22 = Custom!(
   uv=@MaterialExpressionNamedRerouteUsage_14
   v=@MaterialExpressionNamedRerouteUsage_15
   Steps=@MaterialExpressionLinearInterpolate_3
   Tex=@MaterialExpressionTextureObject_0
   Sample=@MaterialExpressionConstant_32
   Index=@MaterialExpressionConstant_15
   Depth=@MaterialExpressionScalarParameter_8
   PixelPos=@MaterialExpressionScreenPosition_1.PixelPosition
   minSignalDepth=@MaterialExpressionConstant_10
   maxSignalDepth=@MaterialExpressionConstant_9
   Offset=@MaterialExpressionScalarParameter_9
   Flow=@MaterialExpressionConstant3Vector_4
   FlowAmount=@MaterialExpressionConstant_29
  ) {
   code="if(Steps==0 || Sample < 0.001){\r\nreturn float4(0,0,0,0);\r\n}\r\nfloat2 uvOut = float2(0,0);\r\nfloat4 outColor = float4(0,0,0,0);\r\nfloat alpha = 0;\r\nfloat i=0;\r\nfloat4 intersect= float4(0,0,0,0);\r\nfloat intersectAccum = 0;\r\nfloat2 Random = Rand3DPCG16( uint3( PixelPos, View.StateFrameIndexMod8) ).xy;\r\nwhile (i<Steps){\r\n    float jitter = Hammersley( i, Steps, Random );\r\n    float h = Depth*(i+jitter)/Steps;\r\n\tfloat2 uvCurrent = uv+v.xy*(Offset-Depth+h)+(Flow-float2(.0,.0))*FlowAmount*((i+jitter)/Steps);\r\n\tintersect = ((((Texture2DSample(Tex,GetMaterialSharedSampler(TexSampler,View.MaterialTextureBilinearWrapedSampler),uvCurrent)).w)) - minSignalDepth)  / (maxSignalDepth - minSignalDepth)*Depth;\r\n\tif( h < intersect.a )\r\n\t{\t\r\n         uvOut = uvCurrent+intersect.xyz*.01;\r\n\t\t\talpha = 1;\r\n         intersectAccum = h;\r\n\t}\r\n\ti++;\r\n}\r\n\r\nOutUV = uvOut;\r\nOutDepth = intersectAccum;\r\noutColor =  Texture2DSample(Tex,GetMaterialSharedSampler(TexSampler,View.MaterialTextureBilinearWrapedSampler),uvOut);\r\nreturn float4(outColor.xyz,alpha);"
   description=SimpleParralax
   outputType=CMOT_Float4
   additionalOutputs [{outputName=OutUV, outputType=CMOT_Float2}, {outputName=OutDepth, outputType=CMOT_Float1}]
   inputs [
    {inputName=uv, input {expression=&self:MaterialExpressionNamedRerouteUsage_14}}
    {inputName=v, input {expression=&self:MaterialExpressionNamedRerouteUsage_15}}
    {inputName=Steps, input {expression=&self:MaterialExpressionLinearInterpolate_3}}
    {inputName=Tex, input {expression=&self:MaterialExpressionTextureObject_0}}
    {inputName=Sample, input {expression=&self:MaterialExpressionConstant_32}}
    {inputName=Index, input {expression=&self:MaterialExpressionConstant_15}}
    {inputName=Depth, input {expression=&self:MaterialExpressionScalarParameter_8}}
    {
     inputName=PixelPos
     input {expression=&self:MaterialExpressionScreenPosition_1, outputIndex=1}
    }
    {inputName=minSignalDepth, input {expression=&self:MaterialExpressionConstant_10}}
    {inputName=maxSignalDepth, input {expression=&self:MaterialExpressionConstant_9}}
    {inputName=Offset, input {expression=&self:MaterialExpressionScalarParameter_9}}
    {
     inputName=Flow
     input {expression=&self:MaterialExpressionConstant3Vector_4, mask=1, maskB=1, maskG=1, maskR=1}
    }
    {inputName=FlowAmount, input {expression=&self:MaterialExpressionConstant_29}}
   ]
  }
  Multiply(A=TextureSample(UVs=@n22.OutUV) {texture=&T_UtopiaCity_ParallaxGrass}.R)
  Multiply(A=TextureCoordinate(), B=ScalarParameter() {defaultValue=1.5, parameterName=Stripes_UV0_Mult})
  NamedRerouteDeclaration(None=Constant() {r=1}) {name=GrassSample, nodeColor=<rgba 0 1 0.9531 1>}
  NamedRerouteDeclaration(None=ComponentMask(None=@n22.return) {a=true}) {name=GrassBladeOpacity, nodeColor=<rgba 1 0 0.9141 1>}
  n23 = StaticSwitch(True=@n15.Parallax UVs, False=NamedRerouteUsage(), Value=StaticBoolParameter() {parameterName=UseParallax})
  n24 = NamedRerouteUsage()
  n25 = TextureSample(UVs=QualitySwitch(Default=@n23, Low=@n24, High=@n23, Medium=@n24, Epic=@n23), Tex=@n2) {samplerSource=SSM_Wrap_WorldGroupSettings}
  NamedRerouteDeclaration(None=@n25.A) {name=PDO_Height, nodeColor=<rgba 1 0 0.9141 1>}
  NamedRerouteDeclaration(None=ComponentMask(None=@n22.return) {r=true}) {name=GrassBladeColor_R_2, nodeColor=<rgba 1 0 0.9141 1>}
  n26 = WorldPosition() {worldPositionShaderOffset=WPT_CameraRelativeNoOffsets}
  n27 = Length(None=@n26.XYZ)
  n28 = OneMinus(None=MaterialFunctionCall(Input=@n27, Input Low=Constant() {r=500}, Input High=Constant() {r=3000}) {materialFunction=&RemapValue_0-1}.Result)
  n29 = Multiply(A=@n28, B=@n28)
  NamedRerouteDeclaration(None=Multiply(A=@n29, B=@n29)) {name=GrassShadow, nodeColor=<rgba 1 0 0.2109 1>}
  NamedRerouteDeclaration(None=@n15.Shadow) {name=POM_Shadow, nodeColor=<rgba 1 0 0.9141 1>}
  n30 = OneMinus(None=MaterialFunctionCall(Input=@n27, Input Low=Constant() {r=500}, Input High=Constant() {r=5000}) {materialFunction=&RemapValue_0-1}.Result)
  NamedRerouteDeclaration(None=@n30) {name=FadeSteps, nodeColor=<rgba 0 1 0.9297 1>}
  NamedRerouteDeclaration(None=ComponentMask(None=@n22.return) {g=true}) {name=GrassBladeColor_G_2, nodeColor=<rgba 1 0 0.9141 1>}
  NamedRerouteDeclaration(None=@n25.R) {name=PDO_GrassBladeColor_R, nodeColor=<rgba 1 0 0.9141 1>}
  NamedRerouteDeclaration(None=@n25.G) {name=PDO_GrassBladeColor_G, nodeColor=<rgba 1 0 0.9141 1>}
  n31 = ComponentMask(None=@n22.OutDepth) {b=true}
  n32 = AppendVector(A=ComponentMask(None=@n22.OutUV) {g=true, r=true}, B=Multiply(A=OneMinus(None=@n31), B=@MaterialExpressionScalarParameter_8))
  n33 = Constant()
  n34 = OneMinus(None=@MaterialExpressionScalarParameter_9) {desc=ReferencePlane}
  n35 = MaterialFunctionCall!(Vector to Transform=CameraVectorWS(), Z Vector=VertexNormalWS(), Center Location) {materialFunction=/Engine/Functions/Engine_MaterialFunctions02/Math/TransformToZVector}
  n36 = Divide(A=Multiply(A=ComponentMask(None=@n35.Result) {g=true, r=true}) {constB=-1}, B=ComponentMask(None=@n35.Result) {b=true}) {constB=1}
  n37 = Add(A=NamedRerouteUsage(), B=Multiply(A=Multiply(A=@n34) {constB=-1}, B=Multiply(A=@MaterialExpressionScalarParameter_8, B=@n36)))
  n38 = Abs(None=DDX(Value=@n37))
  n39 = WorldPosition()
  n40 = DDX(Value=@n39.XYZ)
  n41 = Abs(None=DDY(Value=@n37))
  n42 = DDY(Value=@n39.XYZ)
  n43 = Max(
   A=Divide(
    A=MaterialFunctionCall(Vector 2=@n38) {materialFunction=/Engine/Functions/Engine_MaterialFunctions02/Utility/VectorLength}.V2 Length
    B=MaterialFunctionCall(Vector 3=@n40) {materialFunction=&VectorLength}.V3 Length
   )
   B=Divide(
    A=MaterialFunctionCall(Vector 2=@n41) {materialFunction=&VectorLength}.V2 Length
    B=MaterialFunctionCall(Vector 3=@n42) {materialFunction=&VectorLength}.V3 Length
   )
  )
  n44 = Abs(None=DotProduct(
    A=MaterialFunctionCall() {materialFunction=&CameraDirectionVector}.Result
    B=CameraVectorWS()
   ))
  NamedRerouteDeclaration(None=Divide(A=Distance(A=@n32, B=@n33), B=Divide(A=@n43, B=@n44)) {constB=1}) {name=PDO_2, nodeColor=<rgba 1 0 0.9141 1>}
  NamedRerouteDeclaration(None=@n15.Pixel Depth Offset) {name=PDO, nodeColor=<rgba 1 0 0.9141 1>}
  n45 = Constant()
  n46 = Constant() {r=1.5}
  n47 = MaterialFunctionCall!(Coordinates, R_01_Scale, R_02_Scale, R_03_Scale, R_04_Scale, G_01_Scale, G_02_Scale, G_03_Scale, G_04_Scale) {materialFunction=/Game/Athena/Helios/Environments/Landscape/Materials/MaterialFunctions_Utilities/MF_VariationMasks2}
  NamedRerouteDeclaration(None=Saturate(None=LinearInterpolate(A=@n45, B=@n46, Alpha=@n47.G_01))) {name=SG_Noise01, nodeColor=<rgba 0.5 0 1 1>}
  n48 = Saturate(None=LinearInterpolate(A=Constant() {r=-0.25}, B=Constant() {r=1.5}, Alpha=@n47.G_03))
  n49 = Constant() {r=0.65}
  NamedRerouteDeclaration(None=Saturate(None=Multiply(A=@n48, B=@n49))) {name=SG_Noise03, nodeColor=<rgba 0.5 0 1 1>}
  n50 = Constant() {r=4}
  NamedRerouteDeclaration(None=Saturate(None=Multiply(A=@n47.G_04, B=@n50))) {name=SG_Noise04, nodeColor=<rgba 0.5 0 1 1>}
  NamedRerouteDeclaration(None=@n22.OutDepth) {name=GrassBladeHeight, nodeColor=<rgba 0.4766 1 0 1>}
  n51 = Multiply(A=@n30, B=@n30)
  n52 = Constant() {r=110}
  n53 = Constant() {r=320}
  NamedRerouteDeclaration(None=Multiply(
    A=Multiply(A=@n51, B=@n51)
    B=MaterialFunctionCall(Input=@n27, Input Low=@n52, Input High=@n53) {materialFunction=&RemapValue_0-1}.Result
   )) {name=Grass fade, nodeColor=<rgba 0.0078 1 0 1>}
  n54 = WorldPosition() {worldPositionShaderOffset=WPT_ExcludeAllShaderOffsets}
  n55 = ScalarParameter() {defaultValue=600, parameterName=Grass_UV_WorldScale}
  n56 = TextureCoordinate()
  n57 = ScalarParameter() {defaultValue=10, parameterName=Grass_UV_Scale}
  n58 = StaticBoolParameter() {defaultValue=true, parameterName=UseWorldCoords}
  NamedRerouteDeclaration(None=StaticSwitch(
    True=Divide(A=@n54.XY, B=@n55)
    False=Multiply(A=@n56, B=@n57)
    Value=NamedRerouteDeclaration(None=@n58) {name=UseWorldCoords, nodeColor=<rgba 1 0 0.8906 1>}
   )) {name=UV, nodeColor=<rgba 1 0 0.8906 1>}
  n59 = CameraVectorWS()
  n60 = Normalize(VectorInput=@n59)
  n61 = DotProduct(A=@n60, B=VertexNormalWS())
  NamedRerouteDeclaration(None=LinearInterpolate(A=@n59, B=Divide(A=@n60, B=@n61)) {constAlpha=0.3}) {name=View, nodeColor=<rgba 1 0.7031 0 1>}
  NamedRerouteDeclaration(None=Multiply(A=NamedRerouteUsage(), B=NamedRerouteUsage())) {name=GrassPresence, nodeColor=<rgba 0.3359 0 1 1>}
  NamedRerouteDeclaration(None=LinearInterpolate(A=ScalarParameter() {defaultValue=0.03, parameterName=DepthMin}, B=@n18, Alpha=NamedRerouteUsage()) {constA=1, constB=8}) {name=GrassDepth, nodeColor=<rgba 0.1016 1 0 1>}
  n62 = Constant() {r=2}
  n63 = Multiply(A=NamedRerouteUsage(), B=@n17)
  n64 = SetMaterialAttributes!(
   MaterialAttributes
   Base Color=LinearInterpolate(A=Multiply(A=@n62, B=@n63) {constB=1.4}, B=@n63, Alpha=NamedRerouteUsage())
   Specular=Constant() {r=0.1}
   Roughness=Constant() {r=1}
   Opacity Mask=NamedRerouteUsage()
   Normal=VertexNormalWS()
   Opacity=Constant() {r=0.5}
   Subsurface Color=Constant()
  ) {attributeSetTypes [
    #1
    #2
    #3
    #4
    #5
    #6
    #7
   ]}
  NamedRerouteDeclaration(None=@n64) {name=Grass Material, nodeColor=<rgba 0 0.0859 1 1>}
  NamedRerouteUsage()
  NamedRerouteUsage()
  NamedRerouteUsage()
  NamedRerouteUsage()
  NamedRerouteUsage()
  NamedRerouteUsage()
  NamedRerouteUsage()
  n65 = VectorParameter() {parameterName=Color1, defaultValue=<rgba 0.0556 0.172 0.0129 1>}
  n66 = LinearInterpolate(
   A=LinearInterpolate(A=@n65.RGB, B=VectorParameter() {parameterName=Color2, defaultValue=<rgba 0.0988 0.3 0.025 1>}.RGB, Alpha=NamedRerouteUsage())
   B=VectorParameter() {parameterName=Color3, defaultValue=<rgba 0.0222 0.15 0.0165 1>}.RGB
   Alpha=NamedRerouteUsage()
  )
  QualitySwitch(Default=@n66, Low=@n65.RGB, High=@n66, Medium=@n65.RGB, Epic=@n66)
  Saturate(None=Multiply(B=VectorParameter() {group=GrassShort, parameterName=GrassShort_ColorMult, defaultValue=<rgba 1 1 1 1>}.RGB))
  StaticBool()
  n67 = ScalarParameter() {defaultValue=70, parameterName=Stripes_Offset}
  n68 = WorldPosition() {worldPositionShaderOffset=WPT_ExcludeAllShaderOffsets}
  n69 = Abs(None=ScalarParameter() {defaultValue=350, parameterName=Stripes_Scale})
  n70 = TextureSample(UVs=Divide(A=Add(A=@n67, B=@n68.XY), B=Multiply(A=@n69) {constB=-1})) {texture=/Game/Global/Textures/Default/Masks_and_Gradients/gradient_stasis}
  n71 = ScalarParameter() {defaultValue=20, parameterName=Stripes_Mult}
  n72 = Saturate(None=Add(A=Multiply(A=@n70.R, B=@n71), B=ScalarParameter() {defaultValue=0.875, parameterName=Stripes_Add}))
  n73 = LinearInterpolate(
   A=VectorParameter() {group=BiomeColors, parameterName=SG_ColorA, defaultValue=<rgba 0.1231 0.3 0.0268 1>}.RGB
   B=VectorParameter() {group=BiomeColors, parameterName=SG_ColorB, defaultValue=<rgba 0.1072 0.3 0.024 1>}.RGB
   Alpha=NamedRerouteUsage()
  )
  n74 = VectorParameter() {group=BiomeColors, parameterName=SG_ColorC, defaultValue=<rgba 0.1665 0.4 0.0249 1>}
  n75 = NamedRerouteUsage()
  n76 = ScalarParameter() {defaultValue=0.2, parameterName=BasicGrass_G_LerpA}
  n77 = ScalarParameter() {defaultValue=2.5, parameterName=BasicGrass_G_LerpB}
  n78 = NamedRerouteUsage()
  n79 = TextureSample(UVs=@n78) {texture=/Hermes/Sets/UtopiaCity/Textures/T_UtopiaCity_Grass_Floor_C_D}
  n80 = Multiply(
   A=LinearInterpolate(
    A=VectorParameter() {group=BiomeColors, parameterName=SG_ColorD, defaultValue=<rgba 0.1059 0.35 0.024 1>}.RGB
    B=LinearInterpolate(A=@n73, B=@n74.RGB, Alpha=@n75)
    Alpha=NamedRerouteUsage()
   )
   B=Saturate(None=LinearInterpolate(A=@n76, B=@n77, Alpha=@n79.G))
  )
  n81 = Constant()
  n82 = SetMaterialAttributes!(
   MaterialAttributes
   Base Color=Multiply(A=@n72, B=Saturate(None=@n80))
   Specular=TextureSample(UVs=@n78) {texture=/Hermes/Sets/UtopiaCity/Textures/T_UtopiaCity_Grass_Floor_C_H}.R
   Roughness=Constant() {r=0.5}
   Normal=MaterialFunctionCall!(
    Normals=TextureSample(UVs=@n78) {samplerType=SAMPLERTYPE_Normal, texture=/Hermes/Sets/UtopiaCity/Textures/T_UtopiaCity_Grass_Floor_C_N}.RGB
    TransformToTangentSpace
    Normal Strength=Constant() {r=1}
   ) {materialFunction=/Game/Athena/Apollo/Environments/Landscape/Materials/Biome/CommonMaterials/MF_HQ_WorldAlignedNormal}.Result
   Subsurface Color=@n81
   Pixel Depth Offset=@n81
   World Position Offset=@n81
  ) {attributeSetTypes [
    #1
    #2
    #3
    #5
    #7
    #8
    #9
   ]}
  n83 = Add(A=NamedRerouteUsage(), B=ScalarParameter() {defaultValue=1, parameterName=ColorHeightAdd})
  n84 = Multiply(A=Multiply(A=@n66, B=Saturate(None=@n83)), B=@n72)
  n85 = VectorParameter() {parameterName=ShadowOffset, defaultValue=<rgba 0 0 -100 1>}
  n86 = SetMaterialAttributes!(
   MaterialAttributes
   Base Color=@n84
   Specular=ScalarParameter() {defaultValue=0.1, parameterName=FinalSpec}
   Roughness=ScalarParameter() {defaultValue=1, parameterName=FinalRough}
   Normal=VertexNormalWS()
   Opacity=ScalarParameter() {defaultValue=0.5, parameterName=Opacity}
   Subsurface Color=Multiply(A=@n84, B=ScalarParameter() {defaultValue=0.15, parameterName=SSColorMult})
   Pixel Depth Offset=StaticSwitch(True=NamedRerouteUsage(), False=Constant(), Value=StaticBoolParameter() {defaultValue=true, parameterName=UsePDO})
   World Position Offset=ShadowReplace(Default=Constant3Vector(), Shadow=Transform(None=@n85.RGB) {transformSourceType=TRANSFORMSOURCE_Local})
  ) {attributeSetTypes [
    #1
    #2
    #3
    #5
    #6
    #7
    #8
    #9
   ]}
  StaticSwitch(True=@n82, False=@n86, Value=StaticBoolParameter() {parameterName=BasicGrass})
  Subtract(A=@n26.XYZ, B=CameraPositionWS())
  TextureSample() {samplerType=SAMPLERTYPE_LinearColor, texture=/Game/Global/Textures/Default/Flowmaps/T_TorusFlowmap}
  TextureSample() {
   samplerType=SAMPLERTYPE_LinearColor
   texture=/Game/Creative/Sets/PropSets/Primitives/Customization/Textures/GalacticSwirl/T_CM_GalacticSwirl_Flowmap
  }
  TextureSample() {samplerType=SAMPLERTYPE_LinearColor, texture=/Game/VisualThreatManager/StormVisuals/LargeClouds/T_Swirl_Flowmap}
  WorldPosition() {worldPositionShaderOffset=WPT_CameraRelativeNoOffsets}
 ]
}
